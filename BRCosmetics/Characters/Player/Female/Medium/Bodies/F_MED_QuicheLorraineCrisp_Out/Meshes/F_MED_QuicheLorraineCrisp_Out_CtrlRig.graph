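# Control Rig for the QuicheLorraineCrisp medium female body cosmetic, adding corrective deformation and necklace motion on top of the skeleton.
# Setup imports the hierarchy from the skeleton with virtual bones.
# Execution runs the shared DeformRig function with per-region toggles, curves and offsets for areas such as arms, groin, glutes and knees.
# A weighted multi-parent constraint then drives base_necklace from upper-body bones including spine, neck, clavicles and head, followed by a slight translation offset.
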ControlRigBlueprint /BRCosmetics/Characters/Player/Female/Medium/Bodies/F_MED_QuicheLorraineCrisp_Out/Meshes/F_MED_QuicheLorraineCrisp_Out_CtrlRig
graphs [
 RigVMFunctionLibrary {links [], nodes []}
 RigVMModel {nodes [
   n1 = &RigUnit_PrepareForExecution()
   &RigUnit_HierarchyImportFromSkeleton(NameSpace=None, bIncludeCurves=False, bIncludeMeshSockets=False, bIncludeVirtualBones=True, ExecutePin=@n1.ExecutePin)
   RigUnit_BeginExecution = &RigUnit_BeginExecution()
   FN_DeformRig_FNC_1 = RigVMFunctionReferenceNode(
    DeformRig_ Curve="(deformRig_l_upperarm_up_alpha_72_#1=False,deformRig_r_upperarm_up_alpha_73_#2=False,deformRig_l_upperarm_fr_alpha_74_#3=False,deformRig_r_upperarm_fr_alpha_75_CB51B8 ...(truncated)"
    DeformRig_Bool="(upperarm_1_#4=True,elbow_3_#5=True,wrist_7_#6=True,groin_8_#7=True,glute_17_#8=True,knee_9_40C408324CA261 ...(truncated)"
    DeformRig_Debug=(upperarm_debug_18_#4=False,wrist_debug_20_#6=False,lat_debug_26_#9=False,groinGlute_debug_27_#7=False)
    DeformRig_Value="(upperarm_up_18_#10=(Rotation=(X=0,Y=0,Z=0,W=1),Translation=(X=0,Y=0,Z=8),Scale3D=(X=0.7,Y=1,Z=1)),upperarm_fwd_fr_19_#11=(Rotation=(X=0,Y=0,Z=0,W=1),Translation=(X=-4,Y=10,Z=0),Scale3D=(X=0.5,Y=1,Z ...(truncated)"
    ExecuteContext=@RigUnit_BeginExecution.ExecutePin
   )
   > &RigUnit_ParentConstraint(
    AdvancedSettings=(RotationOrderForFilter=XZY)
    Child=(Type=Bone,Name="base_necklace")
    Parents="((Item=(Type=Bone,Name=\"head\"),Weight=0.0013),(Item=(Type=Bone,Name=\"clavicle_r\"),Weight=0.0053),(Item=(Type=Bone,Name=\"clavicle_l\"),Weight=0.0054),(Item=(Type=Bone,Name=\"neck_02\"),Weight=0.0141),(Item=(Type=Bone,Name=\"spine_04\"),Weight=0.104),(Item=(Type= ...(truncated)"
    Weight=1
   )
   > &RigUnit_ModifyTransforms(
    ItemToModify=((Item=(Type=Bone,Name="base_necklace"),Transform=(Rotation=(X=0,Y=0,Z=0,W=1),Translation=(X=0,Y=0,Z=0.1),Scale3D=(X=1,Y=1,Z=1))))
    Weight=1
    WeightMinimum=0
   )
  ]}
]
